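# Masked subsurface material for Hera frozen lakes, blending cracked ice with snow and matching physical materials.
# Ice combines world-projected surface textures, parallax-occlusion cracks and offset sub-ice gravel, lit partly through skylight-tinted emissive.
# Vertex red and a height texture drive snow edges; RVT height cuts a dithered landscape blend, with vertex blue as override.
# Quality, platform and HLOD-proxy switches simplify shading, while world position offset raises the surface for high-altitude cameras.
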Material /Hera/Terrain/Lake/Materials/M_Hera_FrozenLake
graph {
 properties {
  bCastRayTracedShadows=false
  bTangentSpaceNormal=false
  bUseMaterialAttributes=true
  bUsedWithNanite=true
  bUsedWithSplineMeshes=true
  blendMode=BLEND_Masked
  ditherOpacityMask=true
  pixelDepthOffsetMode=PDOM_Legacy
  shadingModel=MSM_Subsurface
  usedShadingModels=Subsurface
 }
 nodes [
  n1 = ComponentMask(None=PixelNormalWS()) {b=true}
  n2 = Saturate(None=Multiply(A=Power(Base=@n1) {constExponent=4}) {constB=2})
  n3 = ScalarParameter() {
   desc=Set this to 1 for full snow meshes like snow drifts. This overrides vertex color and vertex normal Z masking.
   group=--- F3 | Landscape
   parameterName=TopSnow_FULL_Snow
   sortPriority=-2
  }
  Add(A=@n2, B=@n3)
  n4 = Divide(
   A=WorldPosition() {worldPositionShaderOffset=WPT_ExcludeAllShaderOffsets}.XY
   B=Abs(None=ScalarParameter() {defaultValue=1024, group=Sub Ice, parameterName=Sub Ice Detail Size})
  )
  n5 = TextureSample(UVs=@n4) {samplerSource=SSM_Wrap_WorldGroupSettings, texture=/Game/Environments/Asteria/Terrain/Textures/T_Asteria_RockBase_01_D}
  n6 = LinearInterpolate(A=@n5.B) {constB=0.5}
  n7 = ScalarParameter() {defaultValue=0.156, group=Cracks, parameterName=Crack Depth, sortPriority=2}
  n8 = LinearInterpolate(Alpha=MaterialFunctionCall() {materialFunction=&DitherTemporalAA}.Result) {constA=0.5}
  n9 = Divide(
   A=WorldPosition() {worldPositionShaderOffset=WPT_ExcludeAllShaderOffsets}.XY
   B=Abs(None=ScalarParameter() {defaultValue=1382.0529, group=Cracks, parameterName=Crack Size, sortPriority=1})
  )
  n10 = MaterialFunctionCall!(
   Heightmap Texture=TextureObject() {samplerType=SAMPLERTYPE_Masks, texture=/Game/Environments/Asteria/Terrain/Textures/T_Asteria_Ice_Cracks_M}
   Height Ratio=Multiply(A=@n6, B=Multiply(A=@n7, B=@n8))
   Min Steps=LinearInterpolate(A=ScalarParameter() {defaultValue=8, group=Cracks, parameterName=Prallax Min Steps}, Alpha=NamedRerouteUsage()) {constB=2}
   Max Steps=LinearInterpolate(A=ScalarParameter() {defaultValue=32, group=Cracks, parameterName=Prallax Max Steps}, Alpha=NamedRerouteUsage()) {constB=2}
   UVs=@n9
   Heightmap Channel
   Reference Plane=ScalarParameter() {defaultValue=1, group=Cracks, parameterName=Height Center, sortPriority=6}
   -----------------------------
   Use World Coordinates=StaticBool() {value=true}
   Specify Manual Texture Size
   Manual Texture Size
   ------------------------------
   "Render Shadows (Occlusion Mapping)"
   Light Vector
   Shadow Steps
   Shadow Penumbra
   Transform To VertexNormal
  ) {materialFunction=/Engine/Functions/Engine_MaterialFunctions01/Texturing/ParallaxOcclusionMapping}
  n11 = NamedRerouteUsage()
  n12 = TextureSampleParameter2D!(
   UVs=ShadingPathSwitch(
    Default=QualitySwitch(Default=@n10.Parallax UVs, Low=@n11, High=@n11, Medium=@n11)
    Mobile=QualitySwitch(Default=@n10.Parallax UVs, Low=@n11, Medium=@n11)
   )
   "DDX(UVs)"=DDX(Value=@n4)
   "DDY(UVs)"=DDY(Value=@n4)
   Apply View MipBias
  ) {
   mipValueMode=TMVM_Derivative
   parameterName=Param
   samplerSource=SSM_Wrap_WorldGroupSettings
   samplerType=SAMPLERTYPE_Masks
   texture=&T_Asteria_Ice_Cracks_M
  }
  n13 = VectorParameter() {group=Cracks, parameterName=Crack Extrusion Color, sortPriority=4, defaultValue=<rgba 0.0095 0.0474 0.0521 0>}
  n14 = Multiply(A=@n12.RGB, B=@n13.RGB)
  n15 = Divide(
   A=WorldPosition() {worldPositionShaderOffset=WPT_ExcludeAllShaderOffsets}.XY
   B=Abs(None=ScalarParameter() {defaultValue=3894.9006, group=Cracks, parameterName=Crack Noise Size})
  )
  n16 = Multiply(
   A=MaterialFunctionCall(
    In=TextureSample(UVs=@n15) {samplerSource=SSM_Wrap_WorldGroupSettings, texture=&T_Asteria_RockBase_01_D}.G
    Contrast=ScalarParameter() {defaultValue=0.2872, group=Cracks, parameterName=Crack Noise Contrast}
   ) {materialFunction=&CheapContrast}.Result
   B=ScalarParameter() {defaultValue=1, group=Cracks, parameterName=Crack Overall Opacity}
  )
  n17 = TextureSampleParameter2D!(UVs=@n9, Apply View MipBias) {
   parameterName=Surface Cracks
   samplerSource=SSM_Wrap_WorldGroupSettings
   samplerType=SAMPLERTYPE_Masks
   texture=&T_Asteria_Ice_Cracks_M
  }
  n18 = VectorParameter() {group=Cracks, parameterName=Crack Top Color, sortPriority=3, defaultValue=<rgba 0.019 0.0317 0.0365 0>}
  n19 = LinearInterpolate(B=Multiply(A=Add(A=@n17.RGB, B=@n17.RGB), B=@n18.RGB), Alpha=@n16) {constB=0}
  n20 = TextureObject() {samplerType=SAMPLERTYPE_Masks, texture=/Game/Environments/Asteria/Terrain/Textures/T_Asteria_Ice_Ground_01_S}
  n21 = TextureObject() {samplerType=SAMPLERTYPE_Normal, texture=/Game/Environments/Asteria/Terrain/Textures/T_Asteria_Ice_Ground_01_N}
  n22 = MaterialFunctionCall!(
   Diffuse Texture Object=TextureObject() {texture=/Game/Environments/Asteria/Terrain/Textures/T_Asteria_Ice_Ground_01_D}
   Spec Texture Object=@n20
   TextureSize=ScalarParameter() {defaultValue=1438.8214, group=Surface Ice, parameterName=Surface Ice Size}
   --------------------------
   Use seperate settings for Z projection
   Use seperate texture for Z projection
   Texture Size Z projection
   Z Plane Diffuse Texture Object
   Z Plane Spec Texture Object
   "--------------------------      "
   ProjectionTransitionContrast
   Use Texture Alpha Heightmap to Refine the Transistion
   HeightLerpTransitionContrast
   PickTheDiffuseHeightChannel
   TheDiffuseHeightChannel
   "--------------------------       "
   NormalTextureObject=@n21
   World Space Normals
   "--------------------------  "
   Export Float 4 Diffuse
   Export Float 4 Spec
   "--------------------------     "
   WorldPosition
   World Space Normal
  ) {materialFunction=/Engine/Functions/Engine_MaterialFunctions02/Texturing/WorldAlignedTextures_Complex}
  n23 = ShadingPathSwitch(
   Default=VectorParameter() {group=Surface Ice, parameterName=Ice Tint, defaultValue=<rgba 0.0018 0.0156 0.0057 0>}.RGB
   Mobile=VectorParameter() {group=Surface Ice, parameterName=Ice Tint - Mobile, defaultValue=<rgba 0.0018 0.0156 0.0057 0>}.RGB
  )
  n24 = WorldPosition() {worldPositionShaderOffset=WPT_ExcludeAllShaderOffsets}
  n25 = Abs(None=ScalarParameter() {defaultValue=6005.3467, group=Sub Ice, parameterName=Macro Variation Size})
  n26 = Multiply(
   A=ScalarParameter() {defaultValue=0.9, group=Surface Ice, parameterName=Top Ice Opacity}
   B=TextureSample(UVs=Divide(A=@n24.XY, B=@n25)) {samplerSource=SSM_Wrap_WorldGroupSettings, texture=&T_Asteria_RockBase_01_D}.G
  )
  n27 = OneMinus(None=VertexColor().G)
  n28 = QualitySwitch(Default=Multiply(A=@n27) {constB=0.1}, Epic=@n27)
  n29 = LinearInterpolate(
   A=ShadingPathSwitch(
    Default=VectorParameter() {group=Sub Ice, parameterName=Ice Deep Color, defaultValue=<rgba 0.0036 0.0043 0.0052 1>}.RGB
    Mobile=VectorParameter() {group=Sub Ice, parameterName=Ice Deep Color - Mobile, defaultValue=<rgba 0.0036 0.0043 0.0052 1>}.RGB
   )
   B=Saturate(None=Add(A=@n22.Z Diffuse Texture, B=@n23))
   Alpha=Clamp(None=Add(A=@n26, B=@n28) {constB=0}) {maxDefault=2}
  )
  n30 = MaterialFunctionCall!(Normal=@n22.Normal Z Texture, CameraVector=CameraVectorWS()) {materialFunction=/Engine/Functions/Engine_MaterialFunctions02/Utility/CustomReflectionVector}
  n31 = ComponentMask(None=@n30.Reflection Vector) {g=true, r=true}
  n32 = Divide(
   A=Multiply(A=Multiply(A=@n5.G, B=@n6), B=ScalarParameter() {defaultValue=500, group=Sub Ice, parameterName=Sub Ice Depth}) {constB=50}
   B=Abs(None=ComponentMask(None=@n30.Reflection Vector) {b=true})
  )
  n33 = MaterialProxyReplace(
   Realtime=Add(A=@n4, B=Multiply(A=@n31, B=@n32))
   MaterialProxy=Multiply(A=ScalarParameter() {defaultValue=0.1, parameterName=HLOD_DiffuseUV}, B=@n4)
  )
  n34 = MaterialProxyReplace(Realtime=Constant(), MaterialProxy=Constant())
  n35 = Constant() {r=0.1}
  n36 = Multiply(
   A=QualitySwitch(
    Default=TextureSample!(UVs=@n33, Tex, Bias=@n34, Apply View MipBias) {mipValueMode=TMVM_MipBias, texture=/Game/Environments/Asteria/Terrain/Textures/T_Asteria_FrozenGravel_01_D}.R
    Low=@n35
    High=@n35
    Medium=@n35
   )
   B=ShadingPathSwitch(
    Default=VectorParameter() {group=Sub Ice, parameterName=Sub Ice Color, defaultValue=<rgba 0.097 0.4374 0.5938 1>}.RGB
    Mobile=VectorParameter() {group=Sub Ice, parameterName=Sub Ice Color - Mobile, defaultValue=<rgba 0.097 0.4374 0.5938 1>}.RGB
   )
  )
  n37 = Multiply(A=@n12.RGB, B=@n16)
  n38 = LinearInterpolate(
   A=Add(A=Multiply(A=@n14, B=@n16), B=Add(A=@n19, B=@n29))
   B=@n36
   Alpha=Multiply(A=ScalarParameter() {defaultValue=0.2, group=Sub Ice, parameterName=Sub Ice Alpha}, B=OneMinus(None=@n37))
  )
  n39 = ScalarParameter() {defaultValue=4, parameterName=EmissiveBoost}
  n40 = SkyLightEnvMapSample(Direction=Constant3Vector() {constant=<rgba 0 0 1 0>}, Roughness=Constant() {r=1.5})
  n41 = MaterialProxyReplace(Realtime=Multiply(A=Multiply(A=@n38, B=@n39), B=LinearInterpolate(A=@n40) {constAlpha=0.1, constB=0.5}), MaterialProxy=@n36)
  n42 = OneMinus(None=Fresnel(ExponentIn=ScalarParameter() {defaultValue=3, parameterName=Fresnel Fade}, Normal=VertexNormalWS()))
  n43 = GetMaterialAttributes(None=MaterialFunctionCall!(Grass_UseCustomSlopeMask, Grass_CustomSlopeMask) {materialFunction=/Hera/Environment/Terrain/Landscape/Materials/Layers/MF_Hera_StaticMeshLandscapeLayerMatch}.SnowLayer) {attributeGetTypes [#1]}
  n44 = GetMaterialAttributes(None=@n43.MaterialAttributes) {attributeGetTypes [
    #2
    #3
    #4
    #5
    #6
    #7
    #8
   ]}
  n45 = NamedRerouteUsage()
  n46 = Constant() {r=1}
  n47 = ScalarParameter() {defaultValue=1, parameterName=Mobile Specular Adjust}
  n48 = CollectionParameter() {
   collection=&FortniteMaterialParameters
   parameterName=FortniteActiveTimeOfDay
  }
  n49 = MaterialFunctionCall(Float4=@n48) {materialFunction=&BreakOutFloat4Components}
  n50 = Add(A=@n49.G) {constB=-0.75}
  n51 = Saturate(None=Multiply(A=Saturate(None=@n50)) {constB=3.5})
  n52 = Saturate(None=Add(A=Multiply(B=@n51) {constA=0.4}, B=@n49.A))
  n53 = ShadingPathSwitch(Default=@n45, Mobile=Multiply(A=@n45, B=LinearInterpolate(A=@n46, B=@n47, Alpha=@n52) {constA=1}))
  n54 = NamedRerouteUsage()
  n55 = ScalarParameter() {defaultValue=0.75, parameterName=HLOD_IceSpec}
  n56 = ScalarParameter() {defaultValue=0.4, parameterName=HLOD_SnowSpec}
  n57 = Saturate(None=@n54)
  n58 = Constant()
  MaterialExpressionAppend3Vector_0 = MaterialXAppend3Vector!(A=@n58, B=@n58, C=ScalarParameter() {defaultValue=50, parameterName=WPO_RaiseAmount})
  n59 = InverseLinearInterpolate(
   A=ScalarParameter() {defaultValue=20000, parameterName=WPO_Start}
   B=ScalarParameter() {defaultValue=60000, parameterName=WPO_End}
   Value=ComponentMask(None=CameraPositionWS()) {b=true}
  )
  n60 = Multiply(A=Multiply(A=@n41, B=@n41), B=ScalarParameter() {defaultValue=5, group=Surface Ice, parameterName=Emissive}) {constB=0.282}
  n61 = SetMaterialAttributes!(
   MaterialAttributes
   Base Color=LinearInterpolate(A=Multiply(A=@n41, B=@n42), B=@n44.Base Color, Alpha=NamedRerouteUsage())
   Specular=MaterialProxyReplace(Realtime=LinearInterpolate(A=@n53, B=@n44.Specular, Alpha=@n54), MaterialProxy=LinearInterpolate(A=@n55, B=@n56, Alpha=@n57) {constA=1})
   Roughness=LinearInterpolate(A=NamedRerouteUsage(), B=@n44.Roughness, Alpha=NamedRerouteUsage())
   Normal=LinearInterpolate(A=NamedRerouteUsage(), B=@n43.Normal, Alpha=NamedRerouteUsage())
   Metallic=Constant()
   Opacity=LinearInterpolate(B=@n44.Opacity, Alpha=NamedRerouteUsage()) {constA=1}
   Subsurface Color=LinearInterpolate(A=VectorParameter() {parameterName=IceSSS, defaultValue=<rgba 0.0068 0.0228 0.0404 0>}.RGB, B=Constant(), Alpha=NamedRerouteUsage())
   World Position Offset=LinearInterpolate(A=Constant(), B=@MaterialExpressionAppend3Vector_0, Alpha=Saturate(None=@n59))
   Emissive Color=LinearInterpolate(A=Multiply(A=@n42, B=@n60), B=@n44.Emissive Color, Alpha=NamedRerouteUsage()) {constB=0}
  ) {attributeSetTypes [
    #2
    #5
    #7
    #1
    #9
    #4
    #3
    #10
    #6
   ]}
  n62 = MaterialFunctionCall!(In=@n61, "RVT Height Map (Roughness)"=Constant()) {desc="RVT && snow", materialFunction=/Hera/Terrain/Cliffs/Materials/MF_Hera_RVTBlendingCliffRock}
  n63 = WorldPosition()
  n64 = Saturate(None=Divide(A=Subtract(A=@n62.RVT H, B=@n63.Z), B=ScalarParameter() {defaultValue=15, parameterName=RVTHeightBlendSize}) {constB=50})
  n65 = ScalarParameter() {defaultValue=4, parameterName=RVTHeightBlendPow}
  n66 = VertexColor() {desc="for manual fix ups where intersections with meshes reveal that\r\nthere are two surfaces on top of each other when the top surface gets blended out"}
  n67 = GetMaterialAttributes(None=@n62.Result) {attributeGetTypes [#1]}
  n68 = LinearInterpolate(A=NamedRerouteUsage(), B=@n62.RVT N, Alpha=NamedRerouteUsage())
  n69 = Saturate(None=ScalarParameter() {parameterName="UseLandscapeNormal? (0/1)", sliderMax=1})
  n70 = SetMaterialAttributes!(
   MaterialAttributes=@n67.MaterialAttributes
   Opacity Mask=Max(A=Power(Base=@n64, Exp=@n65), B=@n66.B)
   Normal=ShadingPathSwitch(Default=LinearInterpolate(A=@n67.Normal, B=@n68, Alpha=@n69), Mobile=@n67.Normal) {desc=Migrated from FeatureLevelSwitch}
  ) {attributeSetTypes [#11, #1]}
  MaterialFunctionCall!("Xb1,PS4"=@n70, "PC-High&&Epic,XSX,PS5"=@n70, "PC-Perf,Android"=@n62.Result, Switch=@n70, "PC-Low(NotPerf)&&Medium"=@n70) {materialFunction=/Game/Effects/Niagara/MasterMaterials/Functions/Common/PlatformSwitch_MA}
  n71 = VertexColor()
  n72 = Multiply(A=WorldPosition().XY) {constB=-1}
  n73 = ScalarParameter() {defaultValue=512, parameterName=Edge HM Tile}
  n74 = MaterialFunctionCall(
   Transition Phase=Multiply(A=OneMinus(None=@n71.R), B=Constant() {r=1})
   Height Texture=TextureSample!(UVs=Divide(A=@n72, B=@n73), Tex, Apply View MipBias) {texture=/Game/Environments/Asteria/Terrain/Textures/T_Asteria_WindBlownSnow_01_D}.A
   Contrast=ScalarParameter() {defaultValue=0.5, parameterName=Snow Edge Contrast}
  ) {materialFunction=&HeightLerp}
  MaterialProxyReplace(Realtime=@n74.Alpha, MaterialProxy=@n74.Alpha)
  n75 = VertexColor()
  Multiply(A=@n2, B=@n75.R)
  n76 = NamedRerouteUsage()
  n77 = ScalarParameter() {defaultValue=1000, parameterName=POMOptimizationDistance}
  n78 = Multiply(B=@n77) {constA=0.5}
  n79 = Subtract(A=Add(A=PixelDepth().R, B=Multiply(A=@n76, B=@n78)), B=@n77)
  NamedRerouteDeclaration(None=Saturate(None=Divide(A=@n79) {constB=1000})) {name=POMOptimizationMask, nodeColor=<rgba 0 0.4844 1 1>}
  n80 = Constant() {r=0.1}
  n81 = QualitySwitch(Default=ComponentMask(None=@n22.Z Spec) {b=true}, Low=@n80, High=@n80, Medium=@n80)
  NamedRerouteDeclaration(None=@n81) {name=Zspec, nodeColor=<rgba 0.0781 0 1 1>}
  NamedRerouteDeclaration(None=@n9) {name=WorldUV, nodeColor=<rgba 0 1 0.0859 1>}
  n82 = MaterialFunctionCall(
   Spec Texture Object=@n20
   TextureSize=ScalarParameter() {defaultValue=6000, group=Surface Ice, parameterName=Surface Ice Size Far}
   NormalTextureObject=@n21
  ) {materialFunction=&WorldAlignedTextures_Complex}
  n83 = VertexNormalWS()
  n84 = ScalarParameter() {defaultValue=20000, parameterName=Alt Spec Normal Scale Distance}
  n85 = Add(A=PixelDepth().R, B=Multiply(B=@n84) {constA=0.5})
  n86 = Saturate(None=Divide(A=Subtract(A=@n85, B=@n84), B=ScalarParameter() {defaultValue=4000, parameterName=Distance Transition Width}) {constB=1000})
  NamedRerouteDeclaration(None=LinearInterpolate(A=@n22.Normal Z Texture, B=QualitySwitch(Default=@n82.Normal Z Texture, Low=@n83, High=@n83, Medium=@n83), Alpha=@n86)) {name=IceNormal, nodeColor=<rgba 0.0781 0 1 1>}
  NamedRerouteDeclaration(None=Saturate(None=@n2)) {nodeColor=<rgba 0.9688 1 0 1>}
  NamedRerouteDeclaration(None=ComponentMask(None=@n22.Z Spec) {r=true}) {name=IceSpec, nodeColor=<rgba 0.0781 0 1 1>}
  n87 = LinearInterpolate(A=@n81, B=ComponentMask(None=@n82.Z Spec) {b=true}, Alpha=@n86)
  NamedRerouteDeclaration(None=LinearInterpolate(A=QualitySwitch(Default=@n87, Low=@n80, High=@n80, Medium=@n80), Alpha=ScalarParameter() {defaultValue=0.3, parameterName=Roughness Adjust}) {constB=0.1}) {name=IceRough, nodeColor=<rgba 0.0781 0 1 1>}
  OneMinus(None=@n75.R)
  MaterialExpressionNamedRerouteDeclaration_3 = NamedRerouteDeclaration(None=@n74.Alpha) {name=SnowHeightMap, nodeColor=<rgba 0 0.4844 1 1>}
  MaterialExpressionOneMinus_3 = OneMinus(None=@MaterialExpressionNamedRerouteDeclaration_3)
  PhysicalMaterialOutput!(PhysData_Ice=@MaterialExpressionOneMinus_3, Snow=@MaterialExpressionNamedRerouteDeclaration_3) {inputs [
    {
     physicalMaterial=/Game/Balance/Physics/Materials/PhysData_Ice
     input {
      expression=&self:MaterialExpressionOneMinus_3
      inputName=None
      mask=0
      maskA=0
      maskB=0
      maskG=0
      maskR=0
      outputIndex=0
     }
    }
    {
     physicalMaterial=/Game/FortressPhysicalMaterials/Snow
     input {
      expression=&self:MaterialExpressionNamedRerouteDeclaration_3
      inputName=None
      mask=0
      maskA=0
      maskB=0
      maskG=0
      maskR=0
      outputIndex=0
     }
    }
   ]}
  n88 = MaterialFunctionCall() {materialFunction=/Game/Athena/Asteria/Environments/Landscape/Materials/MaterialFunctions/MF_AsteriaLandscapeLayers}
  n89 = GetMaterialAttributes(None=@n88.Snow) {attributeGetTypes [#1, #2]}
  Reroute(None=@n89.Normal)
  n90 = RuntimeVirtualTextureSampleParameter!(UVs, World Position, Mip Bias) {
   group=RVT
   materialType=BaseColor_Normal_Specular_YCoCg
   mipValueMode=RVTMVM_MipBias
   parameterName=RVT
   virtualTexture=/Game/Athena/Apollo/Maps/Landscape/VT/RVT_Apollo_Landscape
  }
  n91 = SetMaterialAttributes!(
   MaterialAttributes
   Base Color=MaterialFunctionCall!(
    NoRVT=GetMaterialAttributes(None=@n88.Snow) {attributeGetTypes [#2, #1]}.Base Color
    RVT=@n90.BaseColor
   ) {materialFunction=/Game/Packages/Fortress_SharedMaterials/Base_Material_Functions/FN_RVT_Switch_Color}.Result
  ) {attributeSetTypes [#2]}
  n92 = MaterialFunctionCall!(Mask=Constant() {r=1}, SlipperySnow_Mask, "Material;"=@n91) {materialFunction=/Game/Athena/Asteria/Environments/Landscape/Materials/MaterialFunctions/MF_SnowPostRVT}
  n93 = SetMaterialAttributes!(
   MaterialAttributes=@n89.MaterialAttributes
   Normal=@n89.MaterialAttributes
   Emissive Color=@n92.SparkleEmissive
   Base Color=Multiply(A=@n89.Base Color, B=ScalarParameter() {defaultValue=0.8, parameterName=AlbedoADjustment})
   Subsurface Color=VectorParameter() {parameterName=SSColorOverride, defaultValue=<rgba 0.0382 0.1221 0.2462 1>}.RGB
  ) {attributeSetTypes [
    #1
    #6
    #2
    #3
   ]}
  Reroute(None=StaticSwitchParameter(
    True=MaterialFunctionCall(NoRVT=@n91, RVT=@n92.SnowMaterial) {materialFunction=&FN_RVT_Switch_Material}.Result
    False=@n93
   ) {defaultValue=true, parameterName=DUseRVT})
  VertexNormalWS()
 ]
}
